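# Translucent effects material for the Bizarre pickaxe, usable on skeletal meshes and Niagara mesh particles, with an animated reveal and glowing circuitry.
# A dynamic parameter sweeps UV sphere masks to drive opacity and a bright edge broken up by digital noise.
# Emissive sums tinted, panning circuit textures, that edge, and a curve-atlas gradient from an object-relative sphere mask with Fresnel-tinted HDR.
# Switch_Glow blends to a simpler flame-textured glow-mesh look.
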Material /BRCosmetics/Weapons/FORT_Melee/Pickaxe_Bizarre/FX/Material/Pickaxe/5/M_Pickaxe_Bizarre
graph {
 properties {
  bUsedWithMeshDeformer=true
  bUsedWithNiagaraMeshParticles=true
  bUsedWithSkeletalMesh=true
  blendMode=BLEND_Translucent
  pixelDepthOffsetMode=PDOM_Legacy
  usedShadingModels=Default Lit
 }
 nodes [
  n1 = ParticleColor()
  n2 = OneMinus(None=ComponentMask(None=TextureCoordinate()) {g=true})
  n3 = Constant() {r=-0.5}
  n4 = LinearInterpolate(Alpha=DynamicParameter().Param1) {constB=3}
  n5 = Constant() {r=-0.05}
  n6 = Subtract(A=SphereMask(A=@n2, B=@n3, Radius=Subtract(A=@n4, B=@n5) {constB=-0.1}), B=SphereMask(A=@n2, B=@n3, Radius=@n4) {hardnessPercent=0})
  n7 = Power(Base=@n6) {constExponent=1.5}
  n8 = TextureSample(UVs=TextureCoordinate() {vTiling=2.5}) {samplerType=SAMPLERTYPE_LinearColor, texture=/Game/Global/Textures/Default/Noise/T_Atlas_DigitalNoise_01}
  n9 = LinearInterpolate(
   A=LinearInterpolate(A=@n8.R, B=@n8.R, Alpha=MaterialFunctionCall() {materialFunction=&DebugTimeSine}.Result)
   B=@n8.B
   Alpha=MaterialFunctionCall(Speed=Constant() {r=0.5}) {materialFunction=&DebugTimeSine}.Result
  )
  n10 = Subtract(A=SphereMask(A=@n2, B=@n3, Radius=@n4), B=Multiply(A=Multiply(A=@n7, B=@n9)) {constB=1.4})
  n11 = OneMinus(None=ComponentMask(None=TextureCoordinate()) {g=true})
  n12 = Constant() {r=-0.5}
  n13 = LinearInterpolate(Alpha=DynamicParameter().Param1) {constB=3}
  LinearInterpolate(
   A=Multiply(A=@n1.A, B=Clamp(None=@n10))
   B=Multiply(A=ParticleColor().A, B=SphereMask(A=@n11, B=@n12, Radius=@n13))
   Alpha=ScalarParameter() {parameterName=Switch_Glow}
  )
  n14 = VectorParameter() {parameterName=UV, defaultValue=<rgba 0.8 3 0 1>}
  n15 = Multiply(A=TextureCoordinate(), B=AppendVector(A=@n14.R, B=@n14.R))
  n16 = TextureSample(UVs=Panner(Coordinate=@n15, Time=Constant() {r=0.98}) {speedX=0.1, speedY=0.1}) {texture=/Game/Effects/Fort_Effects/Textures/T_FX_Circuitboard}
  n17 = Multiply(A=TextureSample(UVs=@n15) {texture=&T_FX_Circuitboard}.R) {constB=0.2}
  n18 = Constant() {r=1.02}
  n19 = Multiply(A=TextureSample(UVs=Panner(Coordinate=@n15, Time=@n18) {speedX=0.1, speedY=0.1}) {texture=&T_FX_Circuitboard}.R) {constB=0.1}
  n20 = Panner(Coordinate=TextureCoordinate() {uTiling=2}) {speedY=-0.3}
  n21 = Multiply(
   A=ScalarParameter() {defaultValue=16, parameterName=CircuitMultiply}
   B=Multiply(
    A=MaterialFunctionCall(X=@n16.R, Y=@n17, Z=@n19) {materialFunction=&MakeFloat3}.Result
    B=TextureSample(UVs=@n20) {samplerType=SAMPLERTYPE_LinearGrayscale, texture=/BRCosmetics/Weapons/FORT_Melee/Pickaxe_Cube_Ninja_Male/FX/T_FluidFlameTile}.R
   )
  )
  n22 = Constant() {r=1}
  n23 = Multiply(
   A=Multiply(A=Desaturation(None=@n21, Fraction=@n22), B=VectorParameter() {parameterName=CircuitColor, defaultValue=<rgba 2 0.2814 1.9826 1>}.RGB)
   B=DynamicParameter().Param2
  )
  n24 = TransformPosition(None=WorldPosition() {worldPositionShaderOffset=WPT_ExcludeAllShaderOffsets}.XYZ) {transformSourceType=TRANSFORMPOSSOURCE_World}
  n25 = TransformPosition(None=ObjectPositionWS()) {transformSourceType=TRANSFORMPOSSOURCE_World}
  n26 = MaterialFunctionCall(Float3=Divide(A=Subtract(A=@n24, B=@n25)) {constB=100}) {materialFunction=&BreakOutFloat3Components}
  n27 = MaterialFunctionCall(
   X=Multiply(A=ScalarParameter() {defaultValue=1, group=Mask, parameterName=MaskSizeX}, B=@n26.R)
   Y=Multiply(A=@n26.G, B=ScalarParameter() {defaultValue=1.5, group=Mask, parameterName=MaskSizeY})
   Z=Multiply(A=@n26.B, B=ScalarParameter() {defaultValue=0.9, group=Mask, parameterName=MaskSizeZ})
  ) {materialFunction=&MakeFloat3}
  n28 = VectorParameter() {group=Mask, parameterName=MaskLoaction, defaultValue=<rgba 0 0 -0.1 1>}
  n29 = MaterialFunctionCall(X=@n28.R, Y=@n28.R, Z=@n28.R) {materialFunction=&MakeFloat3}
  n30 = ScalarParameter() {defaultValue=0.95, group=Mask, parameterName=MaskRadius_b}
  n31 = CurveAtlasRowParameter(CurveTime=Power(
    Base=SphereMask(A=@n27.Result, B=@n29.Result, Radius=@n30) {attenuationRadius=1, hardnessPercent=0}
    Exp=ScalarParameter() {defaultValue=4, group=Mask, parameterName=Mask_Power}
   ) {constExponent=8}) {
   atlas=/Game/Athena/LinearColorCurves/A_Atlas_Turtleneck_Glider
   curve=/Game/Weapons/FORT_Melee/Pickaxe_Bizarre/FX/Curve_Pickaxe_Bizarre_Color
   defaultValue=13
  }
  n32 = Fresnel(ExponentIn=ScalarParameter() {defaultValue=4, parameterName=Fresnel}) {baseReflectFraction=0, exponent=1}
  n33 = VectorParameter() {parameterName=FresnelColor, defaultValue=<rgba 4 0.0625 1.4749 1>}
  n34 = DynamicParameter()
  n35 = ScalarParameter() {defaultValue=100, parameterName=HDRTint}
  n36 = SphereMask(A=@n27.Result, B=@n29.Result, Radius=ScalarParameter() {defaultValue=0.55, group=Mask, parameterName=MaskRadius_a}) {attenuationRadius=0.55, hardnessPercent=0}
  n37 = MaterialFunctionCall(
   LDR Input=Saturate(None=@n31)
   HDR Tint=Multiply(A=@n31, B=Multiply(A=@n32, B=@n33.RGB))
   HDR Intensity Multiplier=LinearInterpolate(B=Multiply(A=@n34.Param3, B=@n35), Alpha=Power(Base=@n36) {constExponent=4.5}) {constA=1}
  ) {materialFunction=&DeriveHDRfromLDR}
  n38 = Power(Base=@n6) {constExponent=30}
  n39 = ScalarParameter() {defaultValue=1.2, parameterName=LightMultiply}
  n40 = Power(Base=@n6) {constExponent=8}
  n41 = TextureSample(UVs=TextureCoordinate() {vTiling=3}) {texture=&T_FX_Circuitboard}
  n42 = Multiply(
   A=VectorParameter() {parameterName=MaskColor, defaultValue=<rgba 43.6864 0 90 1>}.RGB
   B=Add(A=Multiply(A=@n38, B=@n39) {constB=0.1}, B=Multiply(A=@n40, B=@n41.R))
  )
  n43 = Multiply(A=@n1.RGB, B=Add(A=Add(A=@n23, B=@n37.RemappedRange), B=Multiply(A=@n42, B=@n9)))
  n44 = DynamicParameter() {parameterIndex=1, defaultValue=<rgba 0 1 1 1>}
  n45 = ComponentMask(None=TextureCoordinate()) {g=true}
  n46 = Sine(None=Add(A=OneMinus(None=@n45), B=Time()))
  n47 = TextureCoordinate() {uTiling=1.4, vTiling=0.02}
  n48 = LinearInterpolate(Alpha=TextureSample(UVs=Panner(Coordinate=@n47) {speedY=-0.3}) {samplerType=SAMPLERTYPE_LinearGrayscale, texture=&T_FluidFlameTile}.R) {constA=0.18}
  n49 = ComponentMask(None=TextureCoordinate()) {r=true}
  n50 = Saturate(None=Add(A=Power(Base=@n49)) {constB=0.1})
  n51 = Multiply(
   A=LinearInterpolate(A=LinearInterpolate(Alpha=@n46) {constA=0.524}, B=Constant() {r=1}, Alpha=ScalarParameter() {parameterName=Switch_LightofGlowmesh})
   B=Multiply(A=VectorParameter() {parameterName=GlowColor, defaultValue=<rgba 55 0 14.2036 1>}.RGB, B=Multiply(A=@n48, B=@n50))
  )
  n52 = Multiply(A=ScalarParameter() {parameterName=Desaturation}, B=@n50)
  MaterialFunctionCall(Emissive=LinearInterpolate(
    A=Desaturation(None=@n43, Fraction=@n44.Param1)
    B=Desaturation(None=@n51, Fraction=@n52)
    Alpha=ScalarParameter() {parameterName=Switch_Glow}
   )) {materialFunction=&MF_FX_EmissiveAdjustments}
  n53 = TextureCoordinate() {uTiling=1.55, vTiling=0.6}
  Multiply(
   A=@n41.R
   B=TextureSample(UVs=Panner(Coordinate=@n53) {speedY=-1.6}) {samplerType=SAMPLERTYPE_LinearGrayscale, texture=&T_FluidFlameTile}.R
  )
 ]
}
